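# Animation Blueprint for the OatmealSpreadActive backpack cosmetic, adding secondary motion to its straps, chain and charm.
# Two AnimDynamics strap setups branch from the cached input pose and blend on bIsInFrontEnd, then feed the chain and charm simulations.
# Additive bone transforms follow, keyed to skydive float and dive states.
# The EventGraph maps frame rate over 30–60 fps to the alpha of the front-end setup's first strap simulation.
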AnimBlueprint /BRCosmetics/Accessories/FORT_Backpacks/Backpack_M_MED_OatmealSpreadActive/Meshes/M_MED_OatmealSpreadActive_Pack_AnimBP
target_skeleton=&/M_MED_OatmealSpreadActive_Pack_Skeleton
graphs [
 AnimGraph {nodes [
   n1 = Anim Dynamics - Bone: dyn_strap_0(ComponentPose=Local To Component(LocalPose=Use cached pose 'input pose'()), Alpha=1, ExternalForce="0,0,0") {
    angularDampingOverride=0.9999
    angularSpringConstant=1300
    bAngularSpring=true
    bOverrideAngularDamping=true
    bUseGravityOverride=true
    alphaBoolBlend {blendOption=Linear}
    boundBone {boneName=dyn_strap_0}
    componentAppliedLinearAccClamp=<300 1000 2000>
    componentLinearAccScale=<0.5 0.5 0.6>
    componentLinearVelScale=<0.6 0.6 0.6>
    gravityOverride {z=-800}
    physicsBodyDefinitions [{
      boundBone {boneName=dyn_strap_0}
      boxExtents=<7 7 35>
      constraintSetup {angularLimitsMax=<0 5 0>, angularLimitsMin=<-5 0 -30>, angularTarget=<0 0 0>}
      localJointOffset=<1 0 0>
     }]
   }
   n2 = Anim Dynamics - Bone: dyn_strap_1(ComponentPose=@n1, Alpha=1, ExternalForce="0,0,0") {
    angularDampingOverride=1
    angularSpringConstant=1000
    bAngularSpring=true
    bOverrideAngularDamping=true
    bUseGravityOverride=true
    alphaBoolBlend {blendOption=Linear}
    boundBone {boneName=dyn_strap_1}
    componentAppliedLinearAccClamp=<1000 1000 1000>
    componentLinearAccScale=<0.5 0.5 0.5>
    componentLinearVelScale=<0.6 0.6 0.6>
    gravityOverride {z=-980}
    physicsBodyDefinitions [{
      boundBone {boneName=dyn_strap_1}
      boxExtents=<7 7 20>
      constraintSetup {angularLimitsMax=<0 0 0>, angularLimitsMin=<-2 0 -10>, angularTarget=<0 0 0>}
      localJointOffset=<1 0 0>
     }]
   }
   n3 = Component To Local(ComponentPose=@n2)
   n4 = Anim Dynamics - Bone: dyn_strap_0(
    ComponentPose=Local To Component(LocalPose=Use cached pose 'input pose'())
    Alpha=Get Relaxation Fps() -> (Relaxation Fps:double)
    ExternalForce="0,0,0"
   ) {
    angularDampingOverride=1
    angularSpringConstant=1000
    bAngularSpring=true
    bOverrideAngularDamping=true
    bUseGravityOverride=true
    alphaBoolBlend {blendOption=Linear}
    alphaScaleBiasClamp {bInterpResult=true}
    boundBone {boneName=dyn_strap_0}
    componentAppliedLinearAccClamp=<2000 2000 2000>
    componentLinearAccScale=<0.5 0.6 0.6>
    componentLinearVelScale=<0.6 0.6 0.6>
    gravityOverride {z=-980}
    physicsBodyDefinitions [{
      boundBone {boneName=dyn_strap_0}
      boxExtents=<7 7 35>
      constraintSetup {angularLimitsMax=<0 5 3>, angularLimitsMin=<0 0 -3>, angularTarget=<0 0 0>}
      localJointOffset=<1 0 0>
     }]
   }
   n5 = Anim Dynamics - Bone: dyn_strap_1(ComponentPose=@n4, Alpha=1, ExternalForce="0,0,0") {
    angularDampingOverride=1
    angularSpringConstant=1000
    bAngularSpring=true
    bOverrideAngularDamping=true
    bUseGravityOverride=true
    simulationSpace=BoneRelative
    alphaBoolBlend {blendOption=Linear}
    boundBone {boneName=dyn_strap_1}
    componentAppliedLinearAccClamp=<1000 1000 1000>
    componentLinearAccScale=<0.5 0.5 0.5>
    componentLinearVelScale=<0.6 0.6 0.6>
    gravityOverride {z=100}
    physicsBodyDefinitions [{
      boundBone {boneName=dyn_strap_1}
      boxExtents=<7 7 20>
      constraintSetup {angularLimitsMax=<0 0 5>, angularLimitsMin=<0 0 -15>, angularTarget=<0 0 0>}
      localJointOffset=<0.5 1 0>
     }]
    relativeSpaceBone {boneName=base_strap}
   }
   n6 = Component To Local(ComponentPose=@n5)
   n7 = Get bIsInFrontEnd()
   n8 = Anim Dynamics - Bone: dyn_chain_0(
    ComponentPose=Local To Component(LocalPose=Blend Poses by bool(BlendPose_0=@n6, BlendPose_1=@n3, bActiveValue=@n7) {blendType=Linear})
    Alpha=1
    ExternalForce="0,0,0"
   ) {
    angularDampingOverride=1
    angularSpringConstant=1000
    bAngularSpring=true
    bOverrideAngularDamping=true
    bUseGravityOverride=true
    gravityScale=0.35
    simulationSpace=BoneRelative
    alphaBoolBlend {blendOption=Linear}
    boundBone {boneName=dyn_chain_0}
    componentAppliedLinearAccClamp=<200 200 2000>
    componentLinearAccScale=<0.4 0.4 0.4>
    componentLinearVelScale=<0.5 0.5 0.5>
    gravityOverride {z=-980}
    physicsBodyDefinitions [{
      boundBone {boneName=dyn_chain_0}
      boxExtents=<7 7 7>
      constraintSetup {angularLimitsMax=<5 20 0>, angularLimitsMin=<-5 0 -45>, angularTarget=<1 0 0>}
      localJointOffset=<1 0 0>
     }]
    relativeSpaceBone {boneName=root}
   }
   n9 = Anim Dynamics - Bone: dyn_chain_1(ComponentPose=@n8, Alpha=1, ExternalForce="0,0,0") {
    angularDampingOverride=1
    angularSpringConstant=1500
    bAngularSpring=true
    bOverrideAngularDamping=true
    bUseGravityOverride=true
    gravityScale=0.35
    simulationSpace=BoneRelative
    alphaBoolBlend {blendOption=Linear}
    boundBone {boneName=dyn_chain_1}
    componentAppliedLinearAccClamp=<2000 2000 2000>
    componentLinearAccScale=<0.4 0.4 0.5>
    componentLinearVelScale=<0.5 0.5 0.6>
    gravityOverride {z=-600}
    physicsBodyDefinitions [{
      boundBone {boneName=dyn_chain_1}
      boxExtents=<7 7 7>
      constraintSetup {angularLimitsMax=<5 40 50>, angularLimitsMin=<-10 -2 5>, angularTarget=<1 0 0>}
      localJointOffset=<1 0 0>
     }]
    relativeSpaceBone {boneName=pack}
   }
   n10 = Anim Dynamics - Bone: dyn_chain_2(ComponentPose=@n9, Alpha=1, ExternalForce="0,0,0") {
    angularDampingOverride=1
    angularSpringConstant=1000
    bAngularSpring=true
    bOverrideAngularDamping=true
    bUseGravityOverride=true
    bUsePlanarLimit=false
    alphaBoolBlend {blendOption=Linear}
    boundBone {boneName=dyn_chain_2}
    componentAppliedLinearAccClamp=<1000 1000 1000>
    componentLinearAccScale=<0.5 0.5 0.5>
    componentLinearVelScale=<0.6 0.6 0.6>
    gravityOverride {z=-980}
    physicsBodyDefinitions [{
      boundBone {boneName=dyn_chain_2}
      boxExtents=<7 7 7>
      constraintSetup {angularLimitsMax=<0 5 -10>, angularLimitsMin=<0 -15 -50>, angularTarget=<0 0 0>}
      localJointOffset=<1 0 0>
     }]
   }
   n11 = Anim Dynamics - Bone: dyn_charm(ComponentPose=@n10, Alpha=1, ExternalForce="0,0,0") {
    angularDampingOverride=1
    angularSpringConstant=1000
    bAngularSpring=true
    bOverrideAngularDamping=true
    bUseGravityOverride=true
    alphaBoolBlend {blendOption=Linear}
    boundBone {boneName=dyn_charm}
    componentAppliedLinearAccClamp=<1000 1000 1200>
    componentLinearAccScale=<0.6 0.6 0.6>
    componentLinearVelScale=<0.6 0.6 0.6>
    gravityOverride {z=-980}
    physicsBodyDefinitions [{
      boundBone {boneName=dyn_charm}
      boxExtents=<7 7 7>
      constraintSetup {angularLimitsMax=<2 45 20>, angularLimitsMin=<-2 2 -10>, angularTarget=<0 0 0>}
      localJointOffset=<1 0 0>
     }]
   }
   n12 = Get bIsSkydiveFloating()
   n13 = "Transform (Modify) Bone - Bone: pack"!(ComponentPose=@n11, bAlphaBoolEnabled:bool=@n12, Translation="0,0,0", Rotation="0,3,15", Scale) -> (Pose) {
    alphaInputType=Bool
    rotationMode=BMM_Additive
    rotationSpace=BCS_BoneSpace
    translationMode=BMM_Additive
    translationSpace=BCS_BoneSpace
    alphaBoolBlend {blendInTime=0.2, blendOption=Linear, blendOutTime=0.2}
    boneToModify {boneName=pack}
    rotation {roll=15, yaw=3}
   }
   n14 = "Transform (Modify) Bone - Bone: dyn_chain_0"!(ComponentPose=@n13, bAlphaBoolEnabled:bool=@n12, Translation="0,-0.2,-0.2", Rotation="-10,-30,0", Scale="1,1,1") -> (Pose) {
    alphaInputType=Bool
    rotationMode=BMM_Additive
    rotationSpace=BCS_BoneSpace
    translationMode=BMM_Additive
    translationSpace=BCS_BoneSpace
    alphaBoolBlend {blendInTime=1, blendOption=Linear, blendOutTime=0.2}
    boneToModify {boneName=dyn_chain_0}
    rotation {pitch=-10, yaw=-30}
    translation {y=-0.2, z=-0.2}
   }
   n15 = "Transform (Modify) Bone - Bone: dyn_chain_1"!(ComponentPose=@n14, bAlphaBoolEnabled:bool=@n12, Translation="0,0,0", Rotation="-40,10,0", Scale="1,1,1") -> (Pose) {
    alphaInputType=Bool
    rotationMode=BMM_Additive
    rotationSpace=BCS_BoneSpace
    translationMode=BMM_Additive
    translationSpace=BCS_BoneSpace
    alphaBoolBlend {blendInTime=1, blendOption=Linear, blendOutTime=0.2}
    boneToModify {boneName=dyn_chain_1}
    rotation {pitch=-40, yaw=10}
   }
   n16 = "Transform (Modify) Bone - Bone: dyn_chain_2"!(ComponentPose=@n15, bAlphaBoolEnabled:bool=@n12, Translation="0,0,0", Rotation="30,-30,0", Scale="1,1,1") -> (Pose) {
    alphaInputType=Bool
    rotationMode=BMM_Additive
    rotationSpace=BCS_BoneSpace
    translationMode=BMM_Additive
    translationSpace=BCS_BoneSpace
    alphaBoolBlend {blendInTime=1, blendOption=Linear, blendOutTime=0.2}
    boneToModify {boneName=dyn_chain_2}
    rotation {pitch=30, yaw=-30}
   }
   n17 = "Transform (Modify) Bone - Bone: dyn_charm"!(ComponentPose=@n16, bAlphaBoolEnabled:bool=@n12, Translation="0,0,0", Rotation="-10,0,0", Scale="1,1,1") -> (Pose) {
    alphaInputType=Bool
    rotationMode=BMM_Additive
    rotationSpace=BCS_BoneSpace
    translationMode=BMM_Additive
    translationSpace=BCS_BoneSpace
    alphaBoolBlend {blendInTime=1, blendOption=Linear, blendOutTime=0.2}
    boneToModify {boneName=dyn_charm}
    rotation {pitch=-10}
   }
   n18 = Get bIsSkydiveDiveMode()
   n19 = "Transform (Modify) Bone - Bone: dyn_chain_0"(ComponentPose=@n17, bAlphaBoolEnabled=@n18, Translation="0,-0.2,-0.2", Rotation="-10,-30,0", Scale="1,1,1") {
    alphaInputType=Bool
    rotationMode=BMM_Additive
    rotationSpace=BCS_BoneSpace
    translationMode=BMM_Additive
    translationSpace=BCS_BoneSpace
    alphaBoolBlend {blendInTime=0.2, blendOption=Linear, blendOutTime=0.2}
    boneToModify {boneName=dyn_chain_0}
    rotation {pitch=-10, yaw=-30}
    translation {y=-0.2, z=-0.2}
   }
   n20 = "Transform (Modify) Bone - Bone: dyn_chain_1"(ComponentPose=@n19, bAlphaBoolEnabled=@n18, Translation="0,0,0", Rotation="-40,10,0", Scale="1,1,1") {
    alphaInputType=Bool
    rotationMode=BMM_Additive
    rotationSpace=BCS_BoneSpace
    translationMode=BMM_Additive
    translationSpace=BCS_BoneSpace
    alphaBoolBlend {blendInTime=0.2, blendOption=Linear, blendOutTime=0.2}
    boneToModify {boneName=dyn_chain_1}
    rotation {pitch=-40, yaw=10}
   }
   n21 = "Transform (Modify) Bone - Bone: dyn_chain_2"(ComponentPose=@n20, bAlphaBoolEnabled=@n18, Translation="0,0,0", Rotation="0,20,0", Scale="1,1,1") {
    alphaInputType=Bool
    rotationMode=BMM_Additive
    rotationSpace=BCS_BoneSpace
    translationMode=BMM_Additive
    translationSpace=BCS_BoneSpace
    alphaBoolBlend {blendInTime=0.2, blendOption=Linear, blendOutTime=0.5}
    boneToModify {boneName=dyn_chain_2}
    rotation {yaw=20}
   }
   n22 = "Transform (Modify) Bone - Bone: dyn_charm"(ComponentPose=@n21, bAlphaBoolEnabled=@n18, Translation="0,0,0", Rotation="20,0,0", Scale="1,1,1") {
    alphaInputType=Bool
    rotationMode=BMM_Additive
    rotationSpace=BCS_BoneSpace
    translationMode=BMM_Additive
    translationSpace=BCS_BoneSpace
    alphaBoolBlend {blendInTime=0.2, blendOption=Linear, blendOutTime=0.5}
    boneToModify {boneName=dyn_charm}
    rotation {pitch=20}
   }
   Output Pose(Result=Component To Local(ComponentPose=@n22)) {layerGroup=None}
   Save cached pose 'input pose'(Pose=Input Pose()) {cacheName=input pose}
  ]}
 EventGraph {nodes [
   n1 = Event BlueprintUpdateAnimation()
   > n2 = Set Delta Time X(Delta Time X=@n1.DeltaTimeX)
   > Set Relaxation Fps(Relaxation Fps:double=MapRangeClamped(Value=float / float(A=1, B=@n2), InRangeA=30, InRangeB=60)) -> (Output_Get:double)
   TryGetPawnOwner()
  ]}
]
variables [
 Delta Time X {type=number}
 Relaxation Fps {type=number}
]
